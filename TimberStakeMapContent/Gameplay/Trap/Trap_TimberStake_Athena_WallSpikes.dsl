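; A wall-spike trap Blueprint for the TimberStake map, likely derived from a floor spikes trap parent.
; Its SetSpikePosition event scales the spike mesh's relative Y axis by the new position value, keeping X and Z at 1.
; The spikes therefore extend and retract through scaling rather than movement.
; The construction script only runs the parent's.
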(event EventSetSpikePosition
  (params (NewSpikePosition :type Float%20%28double-precision%29))
  ;[1 1 1] SetSpikePosition
  (Transformation|SetRelativeScale3D (Variables|Trap_Floor_Spikes|GetSpike_Mesh) (Math|Vector|MakeVector 1 NewSpikePosition 1.0)))

(fn ConstructionScript
  (|CallParentFunction))
